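# Sound cue for a rift encounter opening, layering close bursts, distant bursts and thunder under the Fort_Gadgets class.
# A mixer blends two random four-variant burst sets, each with its own attenuation, with a louder thunder layer.
# The thunder is pitch-modulated, briefly delayed and separately attenuated, behind a quality-level node whose fourth slot is empty.
# A concurrency override caps simultaneous instances at three.
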SoundCue /JobBoard/Blueprints/VehicleSpawn/RiftVehicleSpawnerActor/Audio/RiftEncounter_Open_Cue
root=Mixer_0
nodes [
 {
  id=Attenuation_0
  children [Delay_0]
  properties {attenuationSettings=&/RiftEncounter_Open_Thunder_Att}
 }
 {
  id=Attenuation_1
  children [Random_0]
  properties {attenuationSettings=&/RiftEncounter_Close_Att}
 }
 {
  id=Attenuation_2
  children [Random_1]
  properties {attenuationSettings=&/RiftEncounter_Distant_Att}
 }
 {id=Delay_0, children [Modulator_0], properties {delayMax=0.25, delayMin=0.25}}
 {id=Mixer_0, children [Attenuation_1, Attenuation_2, QualityLevel_0], properties {inputVolume [0.7,0.7,1]}}
 {id=Modulator_0, children [WavePlayer_1], properties {pitchMax=1.2, pitchMin=0.9, volumeMax=1, volumeMin=1}}
 {id=QualityLevel_0, children [Attenuation_0, Attenuation_0, Attenuation_0, null]}
 {
  id=Random_0
  children [WavePlayer_0,WavePlayer_3,WavePlayer_4,WavePlayer_5]
  properties {bShouldExcludeFromBranchCulling=true, weights [1,1,1,1]}
 }
 {
  id=Random_1
  children [WavePlayer_6, WavePlayer_7, WavePlayer_8, WavePlayer_10]
  properties {bShouldExcludeFromBranchCulling=true, weights [1,1,1,1]}
 }
 {id=WavePlayer_0, properties {soundWaveAssetPtr=&/RiftEncounter_Open_Burst_01}}
 {
  id=WavePlayer_1
  properties {soundWaveAssetPtr=&/RiftEncounter_Open_Thunder_01}
 }
 {
  id=WavePlayer_10
  properties {soundWaveAssetPtr=&/RiftEncounter_Open_Burst_Distant_04}
 }
 {id=WavePlayer_3, properties {soundWaveAssetPtr=&/RiftEncounter_Open_Burst_02}}
 {id=WavePlayer_4, properties {soundWaveAssetPtr=&/RiftEncounter_Open_Burst_03}}
 {id=WavePlayer_5, properties {soundWaveAssetPtr=&/RiftEncounter_Open_Burst_04}}
 {
  id=WavePlayer_6
  properties {soundWaveAssetPtr=&/RiftEncounter_Open_Burst_Distant_01}
 }
 {
  id=WavePlayer_7
  properties {soundWaveAssetPtr=&/RiftEncounter_Open_Burst_Distant_02}
 }
 {
  id=WavePlayer_8
  properties {soundWaveAssetPtr=&/RiftEncounter_Open_Burst_Distant_03}
 }
]
properties {
 bOverrideConcurrency=true
 maxDistance=25000
 soundClassObject=&Fort_Gadgets
 volumeMultiplier=1
 concurrencyOverrides {maxCount=3}
}
