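# Sound cue for the build-up at the start of a rift encounter, part of the rift vehicle spawner's audio.
# A mixer layers a close branch and a slightly quieter distant branch, each with its own attenuation settings.
# Each branch picks at random, with equal weights, from its own build-up variations and is excluded from branch culling.
# It plays in the gadgets sound class with a concurrency override capping simultaneous instances.
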SoundCue /JobBoard/Blueprints/VehicleSpawn/RiftVehicleSpawnerActor/Audio/RiftEncounter_Start_Cue
root=Mixer_0
nodes [
 {
  id=Attenuation_0
  children [Random_0]
  properties {attenuationSettings=&/RiftEncounter_Close_Att}
 }
 {
  id=Attenuation_1
  children [Random_1]
  properties {attenuationSettings=&/RiftEncounter_Distant_Att}
 }
 {id=Mixer_0, children [Attenuation_0, Attenuation_1], properties {inputVolume [1,0.7]}}
 {
  id=Random_0
  children [WavePlayer_1,WavePlayer_2,WavePlayer_3,WavePlayer_4]
  properties {bShouldExcludeFromBranchCulling=true, weights [1,1,1,1]}
 }
 {id=Random_1, children [WavePlayer_5,WavePlayer_6,WavePlayer_7], properties {bShouldExcludeFromBranchCulling=true, weights [1,1,1]}}
 {
  id=WavePlayer_1
  properties {soundWaveAssetPtr=&/RiftEncounter_Start_Build_Up_01}
 }
 {
  id=WavePlayer_2
  properties {soundWaveAssetPtr=&/RiftEncounter_Start_Build_Up_02}
 }
 {
  id=WavePlayer_3
  properties {soundWaveAssetPtr=&/RiftEncounter_Start_Build_Up_03}
 }
 {
  id=WavePlayer_4
  properties {soundWaveAssetPtr=&/RiftEncounter_Start_Build_Up_04}
 }
 {
  id=WavePlayer_5
  properties {soundWaveAssetPtr=&/RiftEncounter_Start_Build_Up_Distant_01}
 }
 {
  id=WavePlayer_6
  properties {soundWaveAssetPtr=&/RiftEncounter_Start_Build_Up_Distant_02}
 }
 {
  id=WavePlayer_7
  properties {soundWaveAssetPtr=&/RiftEncounter_Start_Build_Up_Distant_03}
 }
]
properties {
 bOverrideConcurrency=true
 maxDistance=25000
 soundClassObject=&Fort_Gadgets
 volumeMultiplier=0.7
 concurrencyOverrides {maxCount=3}
}
