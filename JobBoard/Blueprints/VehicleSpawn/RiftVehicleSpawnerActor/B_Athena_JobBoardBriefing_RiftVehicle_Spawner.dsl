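; Job Board Briefing spawner actor that delivers a vehicle to the owner's squad through a rift.
; On the server, an EQS query picks the spot, which the actor moves to and marks for the squad with a map icon.
; Client rift effects play, with vehicle spawn timed to the sound's length, then a quest gameplay event goes to squadmates.
; OnRep_VehicleTypeSpawned optionally applies random mods, pushes back nearby players and sets the spawner's lifespan.
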(event EventBeginPlay
  ;[0 0 0] Begin Play - Spawn the Rift FX at query location
  (Networking|SwitchHasAuthority
    (:Authority
      (Class|FortAsyncActionCurrentPlaylistReady|PlaylistDataReadyAsync
        (:OnReady
          (bind _returnvalue (AI|EQS|RunEQSQuery (Variables|Default|GetQuery%20Template) (Actor|GetOwner)))
          (bind _output_get (Variables|Default|SetTempEQS _returnvalue))
          (AI|EQS|BindEventtoOnQueryFinished _output_get (EventDispatchers|CreateEvent :!event Custom|OnQueryFinished)))))))

(event Custom|OnQueryFinished
  (params (QueryInstance :type EQS%20Query%20Instance%20Object%20Reference) (QueryStatus :type EEnvQueryStatus%20Enum :default "Processing"))
  ;[0 0 0] On Query Finished - Spawn vehicle and fx at query location found
  (bind (_resultlocations _returnvalue) (AI|EQS|GetQueryResultsasLocations QueryInstance))
  (if (and _returnvalue (> (Utilities|Array|Length _resultlocations) 0))
    (Variables|Default|SetChosenQueryLocation (Math|Vector|MakeVector (.x (Utilities|Array|Get(acopy) _resultlocations)) (.y (Utilities|Array|Get(acopy) _resultlocations)) (+ (.z (Utilities|Array|Get(acopy) _resultlocations)) (ScalableFloat|GetValueatLevel (Variables|Default|GetHeightAboveGround) 1.0))))
    (bind (_sweephitresult _returnvalue_2) (Transformation|SetActorLocation (Variables|Default|GetChosenQueryLocation) false true))
    (Utilities|FlowControl|Delay 2.5)
    (else
      _))
  (CallFunction|SpawnRiftFX)
  (CallFunction|SpawnMapMarkerIcon)
  (if _returnvalue_2
    (CallFunction|MarkActorforOwnerSquad self)
    (else
      (Fortnite|FortShippingLog self "JOBBOARDFX: Rift actor location did not set to vehicle spawn location - will not be marked until vehicle spawns")))
  (Utilities|FlowControl|Delay (- (Variables|/Script/Engine.SoundBase|GetDuration (Variables|Default|GetRiftStartSound)) (Variables|Default|GetSoundDurationBuffer)))
  (CallFunction|SpawnVehicle)
  (CallFunction|SendGameplayEventtoInstigatorSquad)
  (CallFunction|RiftCloseFX)
  (Utilities|FlowControl|Delay 2.0)
  (CallFunction|MarkActorforOwnerSquad (Variables|Default|GetVehicleTypeSpawned)))

(event Custom|SpawnRiftFX
  (bind _chosenquerylocation (Variables|Default|GetChosenQueryLocation))
  ;[0 0 0] Rift starting visuals and sounds
  (Networking|SwitchHasAuthority
    (:Remote
      (bind _returnvalue (Niagara|SpawnSystematLocation :SystemTemplate (Variables|Default|GetFX_RiftSpawnIn_Flash) :Location _chosenquerylocation :Scale "0.500000,0.500000,0.500000" :bAutoDestroy true :bAutoActivate true :PoolingMethod "None" :bPreCullCheck true))
      (Variables|Default|SetFXRiftSpawnInInstance _returnvalue)
      (bind _returnvalue_2 (Audio|SpawnSoundatLocation (Variables|Default|GetRiftStartSound) _chosenquerylocation))
      (Variables|Default|SetRiftStartSoundInstance _returnvalue_2)
      (SoundIndicator|StartSoundIndicatorLooping (Variables|Default|GetFortSoundIndicator) _chosenquerylocation (Variables|Default|GetRiftStartSound) 1.0 0)
      (CallFunction|StartAnticipationCameraShake))))

(event Custom|RiftCloseFX
  ;[0 0 0] Rift closing visuals and sounds
  (Networking|SwitchHasAuthority
    (:Remote
      (CameraShake|StopAllCameraShakes (Variables|Default|GetCameraShakeSource) true)
      (Components|Activation|Deactivate (Variables|Default|GetForceFeedback))
      (Audio|Components|Audio|StopDelayed (Variables|Default|GetRiftStartSoundInstance) 0.1)
      (Components|DestroyComponent (Variables|Default|GetFXRiftSpawnInInstance))
      (GameplayCue|ExecuteGameplayCue%28Burst%29Local self (Variables|Default|GetRiftEndGameplayCue) (Ability|GameplayCue|MakeGameplayCueParameters :NormalizedMagnitude 0.0 :RawMagnitude 0.0 :Location (Variables|Default|GetChosenQueryLocation) :Normal "0,0,0" :Instigator (Actor|GetOwner) :GameplayEffectLevel 1 :AbilityLevel 1 :bReplicateLocationWhenUsingMinimalRepProxy false))
      (Utilities|FlowControl|Delay (Variables|/Script/Engine.SoundBase|GetDuration (Variables|Default|GetRiftCloseSound)))
      (SoundIndicator|StopSoundIndicatorLooping (Variables|Default|GetFortSoundIndicator)))))

(event Custom|SendGameplayEventToInstigatorSquad
  (bind _vehicletypespawned (Variables|Default|GetVehicleTypeSpawned))
  ;[0 0 0] When the vehicle is spawned, send the gameplay event associated with the quest objective gate to instigator squad
  (Networking|SwitchHasAuthority
    (:Authority
      (Utilities|IsValid _vehicletypespawned
        (:Is%20Valid
          (bind _squadplayerstates (CallFunction|GetOwnerSquadmates))
          (for _array_element _squadplayerstates
            (Ability|SendGameplayEventtoActor _array_element (Variables|Default|GetQuestGameplayEventTriggerTag) (Utilities|Struct|MakeGameplayEventData "(TagName=\"\")" 0 _vehicletypespawned (Variables|Default|GetMapIcon)))))))))

(event EventEndPlay
  (params (EndPlayReason :type EEndPlayReason%20Enum :default "Destroyed"))
  ;[0 0 0] EndPlay - unbind events
  (|CallParentFunction EndPlayReason)
  (Networking|SwitchHasAuthority
    (:Authority
      (AI|EQS|UnbindEventfromOnQueryFinished (Variables|Default|GetTempEQS) (EventDispatchers|CreateEvent :!event Custom|OnQueryFinished)))))

(event Custom|StartAnticipationCameraShake
  ;[0 0 0] Anticipation camera shake for when the rift opens
  (CameraShake|Start (Variables|Default|GetCameraShakeSource))
  (Components|Activation|Activate (Variables|Default|GetForceFeedback) false))

(fn ApplyVehicleMods
  (params (SpawnedVehicleParameter :type Actor%20Object%20Reference))
  (bind _modtags (LocalVariables|GetModTags))
  ;[0 0 0] Apply mods to the vehicle
  (bind _asfort_athena_vehicle (Utilities|Casting|CastToFortAthenaVehicle SpawnedVehicleParameter))
  ;;[0.125 0.125 0.125] For each "slot" in the map for mods, check if there are mods we can apply
  (Utilities|IsValid _asfort_athena_vehicle
    (:Is%20Valid
      (bind _keys (Utilities|Map|Keys (LocalVariables|GetModTags)))
      (for _array_element _keys
        (if (and (Utilities|Array|IsNotEmpty (GameplayTags|BreakGameplayTagContainer (Utilities|Map|Find _modtags _array_element))) (.ReturnValue (Utilities|Map|Find _modtags _array_element)))
          ;;[0.125 0.125 0.125] Pick a mod at random from the array to apply
          (if (Vehicle|Mod|CanApplyMod _asfort_athena_vehicle (Utilities|Array|RandomArrayItem (GameplayTags|BreakGameplayTagContainer (Utilities|Map|Find _modtags _array_element))))
            (Vehicle|Mod|ApplyVehicleMod _asfort_athena_vehicle (Utilities|Array|RandomArrayItem (GameplayTags|BreakGameplayTagContainer (Utilities|Map|Find _modtags _array_element))) 0)))))))

(fn GetOwnerSquadmates
  (bind _returnvalue (Actor|GetOwner))
  (bind _asfort_player_controller (Utilities|Casting|CastToFortPlayerController _returnvalue))
  (bind _returnvalue_2 (Fortnite|Utilities|AttemptGetPlayerStateforPawn (Game|Player|GetPlayerPawn _asfort_player_controller)))
  (bind _asfort_player_state_athena (Utilities|Casting|CastToFortPlayerStateAthena _returnvalue_2))
  (bind _returnvalue_3 (Class|FortGameStateAthena|GetSquadMemberPlayerStates (Fortnite|Athena|GameState|GetGameStateAthenafromContext _returnvalue) _asfort_player_state_athena true))
  (return _returnvalue_3))

(fn MakeTrackingRadiusForActor
  (params (ActorToTrack :type Actor%20Object%20Reference))
  (bind _squadplayerstates (CallFunction|GetOwnerSquadmates))
  (for _array_element _squadplayerstates
    (bind _returnvalue (RadiusTracker|MakeRadiusTrackingInitializer 0.0 (LocalVariables|GetTrackingMinRadius) (LocalVariables|GetTrackingMaxRadius) 0.0 0.0 (LocalVariables|GetCircleColor) (LocalVariables|GetCircle%20Drawing%20Material) ActorToTrack (Utilities|Array|MakeArray (Fortnite|GetFortPlayerControllerfromActor _array_element))))
    (Fortnite|AddActorTrackingInfo (RadiusTracker|GetGameStateRadiusTrackerComponent) _returnvalue)))

(fn MarkActorForOwnerSquad
  (params (ActorToMark :type Actor%20Object%20Reference))
  (bind _returnvalue (Actor|GetOwner))
  (bind _returnvalue_2 (Fortnite|GetFortPlayerControllerfromActor _returnvalue))
  (bind _returnvalue_3 (Class|FortPlayerController|GetMarkerComponent _returnvalue_2))
  ;[0 0 0] Mark the specified actor for the instigator and squadmates
  (Utilities|IsValid _returnvalue_2
    (:Is%20Valid
      (Utilities|IsValid _returnvalue_3
        (:Is%20Valid
          (Utilities|IsValid ActorToMark
            (:Is%20Valid
              (MarkerComponent|MarkActoronServer _returnvalue_3 ActorToMark true false)
              (return))))))))

(fn OnRep_VehicleTypeSpawned
  ;[0 0 0] Destroy this actor once the vehicle is spawned after a bit of time
  (Networking|SwitchHasAuthority
    (:Authority
      (if (== (ScalableFloat|GetValueatLevel (Variables|Default|GetCanApplyVehicleMod) 1.0) 1.0)
        (CallFunction|ApplyVehicleMods (Variables|Default|GetVehicleTypeSpawned))
        (else
          _))
      (for _array_element (Collision|GetOverlappingActors (Variables|Default|GetKnockbackCollision) "/Script/FortniteGame.FortPlayerPawn" :!class /Script/Engine.PrimitiveComponent)
        ;;[0.1562 0.1562 0.1562] Attempt to Push Back any Players that are nearby
        (Utilities|IsValid (Ability|GetAbilitySystemComponent _array_element)
          (:Is%20Valid
            (Ability|Advanced|GrantandActivateAbilityonSelfwithParams (Ability|GetAbilitySystemComponent _array_element) (Variables|Default|GetPushbackAbility) _array_element (Utilities|Struct|MakeGameplayEventData "(TagName=\"\")" 0 (Variables|Default|GetVehicleTypeSpawned))))))
      (Actor|SetLifeSpan (Variables|Default|GetSpawner%20Life%20Span)))))

(fn SpawnMapMarkerIcon
  (bind _returnvalue (Game|SpawnActorfromClass (Variables|Default|GetMapIconClass) (Math|Conversions|ToTransform(Vector) (Variables|Default|GetChosenQueryLocation)) "AlwaysSpawn" "MultiplyWithRoot" 0 0))
  (bind _output_get (Variables|Default|SetMapIcon _returnvalue))
  (if (Utilities|IsValid (Actor|GetComponentbyClass _output_get "/Script/FortniteGame.FortMiniMapComponent") :!class /Script/Engine.KismetSystemLibrary)
    (MiniMapData|SetMiniMapIndicatorTeam (Actor|GetComponentbyClass _output_get "/Script/FortniteGame.FortMiniMapComponent") (Team|GetActorTeam (Actor|GetOwner)))
    (return)))

(fn SpawnVehicle
  (bind _spawnvehicletype (Variables|Default|GetSpawnVehicleType))
  (bind _returnvalue (ScalableFloat|GetValueasInteger _spawnvehicletype))
  (bind _spawnvehicleclass (Variables|Default|GetSpawnVehicleClass))
  (bind _output (Utilities|Array|Get(acopy) _spawnvehicleclass _returnvalue))
  ;[0 0 0] Handle vehicle specific logic
  ;;[0.1927 0.1927 0.1927 0.8] Get table value of the vehicle at the level - if 0.0, sport car; if 1.0, truck
  (if (Utilities|IsValidClass _output)
    (bind _returnvalue_2 (Game|SpawnActorfromClass _output (Math|Conversions|ToTransform(Vector) (Variables|Default|GetChosenQueryLocation)) "AdjustIfPossibleButAlwaysSpawn" "SelectDefaultAtRuntime" 0 false 0))
    ;;[0.1927 0.1927 0.1927 0.8] Set up vehicle initial values
    (bind _asvalet_sports_car_vehicle (Utilities|Casting|CastToValet_SportsCar_Vehicle _returnvalue_2))
    (Variables|Default|SetVehicleTypeSpawned _asvalet_sports_car_vehicle)
    (Vehicle|WakeUp _asvalet_sports_car_vehicle)
    (Vehicle|Fuel|SetFuel _asvalet_sports_car_vehicle 1e+02)))

(fn ConstructionScript)
